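; Apollo_Vault_Floor is a floor piece Blueprint for the Apollo vault environment sets that refuses death from damage.
; Its ShouldDie function returns false regardless of damage, instigator, causer, test flag or effect context.
; The construction script is empty, and the Blueprint shows no other logic.
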(fn ShouldDie
  (params (Damage :type Float%20%28single-precision%29) (EventInstigator :type Controller%20Object%20Reference) (DamageCauser :type Actor%20Object%20Reference) (bIsTest :type Boolean) (EffectContext :type Gameplay%20Effect%20Context%20Handle%20Structure%20%28by%20ref%29))
  (return false))

(fn ConstructionScript)
