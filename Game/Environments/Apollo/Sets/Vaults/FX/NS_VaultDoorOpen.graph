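# Niagara system for a vault door opening, a brief burst of white translucent sprites pushed out from the door.
# Two emitters, BurstFastCylin and BurstSlowCylin, each spawn an instantaneous burst on a thin half-cylinder rim around the Y axis, offset and drifting toward -Y.
# Drag slows the sprites while curves grow them and fade their alpha.
# The fast layer is short-lived and more opaque, while the slow layer starts smaller and lingers longer.
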NiagaraSystem /Game/Environments/Apollo/Sets/Vaults/FX/NS_VaultDoorOpen
dependencies {
 usedDataInterfaces [&NiagaraDataInterfaceCurve, &NiagaraDataInterfaceColorCurve]
 usedDynamicInputs []
 usedModules [&SystemState]
}
emitters {
 BurstFastCylin {
  emitterUpdateScript [
   &EmitterState(Life Cycle Mode=::NewEnumerator1, Loop Behavior=::NewEnumerator1, Loop Duration=2, UseLoopDelay=false)
   &SpawnBurst_Instantaneous(Spawn Count=12, Use Spawn Probability=false)
  ]
  particleSpawnScript [
   /Niagara/Modules/Spawn/Initialization/InitializeParticle(
    Color=&UniformRangedLinearColor(Minimum=<rgba 1 1 1 0.8>, Maximum=<rgba 1 1 1 0.9>)
    Lifetime=&UniformRangedFloat(Minimum=0.3, Maximum=0.6)
    Mass=&UniformRangedFloat(Minimum=0.5, Maximum=1)
    Module.Write Color=true
    Module.Write Lifetime=true
    Module.Write Mass=true
    Module.Write Material Random=true
    Module.Write Position=false
    Module.Write Scale=false
    Module.Write SpriteRotation=true
    Module.Write SpriteSize=true
    Position=&SimulationPosition()
    Sprite Rotation=&UniformRangedFloat(Minimum=0, Maximum=1)
    Sprite Size=&UniformRangedVector2D(Minimum=<60 60>, Maximum=<120 120>)
   )
   &CylinderLocation(
    Cylinder Height=1
    Cylinder Radius=160
    Hemisphere X=true
    Lathe Profile=DataInterface() {propertyValues {Curve {keys []}}}
    MaskSpawnGroup=false
    Module.Enable Lathe Profile=false
    Module.Override Randomness=false
    Module.Surface Only=true
    Offset=&UniformRangedVector(Minimum=<0 -25 0>, Maximum=<1 -50 1>)
    Orientation Axis:ENiagaraOrientationAxis=::YAxis
    Surface Only Band Thickness=1
    Use Endcaps In Surface Only Mode=false
   )
   &AddVelocity(Velocity=&UniformRangedVector(Minimum=<0 -35 0>, Maximum=<1 -70 1>))
  ]
  particleUpdateScript [
   &Drag(Drag=1.5, Use Linear Drag=true, Use Rotational Drag=true)
   &ParticleState() {module=UpdateAge}
   &ScaleSpriteSize(Scale Factor=&Vector2DFromFloat(Value=&FloatFromCurve(FloatCurve=DataInterface() {propertyValues {Curve {keys [<at -0.0046 1.0113 RCIM_Cubic>,<at 1.0011 1.6623>]}}})))
   &ScaleColor(
    Scale Mode=::NewEnumerator1
    Scale RGBA=&Vector4FromCurve(Vector4Curve=DataInterface() {propertyValues {
       AlphaCurve {keys [<at 0.2568 0.9976 RCIM_Cubic>,<at 0.9977 -0.0026>]}
       BlueCurve {keys [<at -0.0011 0.9976 RCIM_Cubic>,<at 1.0011 0.9976>]}
       GreenCurve {keys [<at -0.0011 0.9976 RCIM_Cubic>,<at 1.0011 0.9976>]}
       RedCurve {keys [<at -0.0011 0.9976 RCIM_Cubic>,<at 1.0011 0.9976 RCIM_Cubic>]}
      }})
    ScaleRGBA=true
   )
   &SolveForcesAndVelocity(Clamp Velocity=false, Limit Acceleration=false)
  ]
  rendererClasses [&NiagaraSpriteRendererProperties]
  renderers [{rendererClass=&NiagaraSpriteRendererProperties}]
 }
 BurstSlowCylin {
  emitterUpdateScript [
   &EmitterState(Life Cycle Mode=::NewEnumerator1, Loop Behavior=::NewEnumerator1, Loop Duration=2, UseLoopDelay=false)
   &SpawnBurst_Instantaneous(Spawn Count=15, Use Spawn Probability=false)
  ]
  particleSpawnScript [
   &InitializeParticle(
    Color=&UniformRangedLinearColor(Minimum=<rgba 1 1 1 0.6>, Maximum=<rgba 1 1 1 0.8>)
    Lifetime=&UniformRangedFloat(Minimum=1, Maximum=1.5)
    Mass=&UniformRangedFloat(Minimum=0.5, Maximum=1)
    Module.Write Color=true
    Module.Write Lifetime=true
    Module.Write Mass=true
    Module.Write Material Random=true
    Module.Write Position=false
    Module.Write Scale=false
    Module.Write SpriteRotation=true
    Module.Write SpriteSize=true
    Position=&SimulationPosition()
    Sprite Rotation=&UniformRangedFloat(Minimum=0, Maximum=1)
    Sprite Size=&UniformRangedVector2D(Minimum=<60 60>, Maximum=<120 120>)
   )
   &CylinderLocation(
    Cylinder Height=1
    Cylinder Radius=160
    Hemisphere X=true
    Lathe Profile=DataInterface() {propertyValues {Curve {keys []}}}
    MaskSpawnGroup=false
    Module.Enable Lathe Profile=false
    Module.Override Randomness=false
    Module.Surface Only=true
    Offset=&UniformRangedVector(Minimum=<0 -25 0>, Maximum=<1 -50 1>)
    Orientation Axis:ENiagaraOrientationAxis=::YAxis
    Surface Only Band Thickness=1
    Use Endcaps In Surface Only Mode=false
   )
   &AddVelocity(Velocity=&UniformRangedVector(Minimum=<0 -35 0>, Maximum=<1 -60 1>))
  ]
  particleUpdateScript [
   &Drag(Drag=1.5, Use Linear Drag=true, Use Rotational Drag=true)
   &ParticleState() {module=UpdateAge}
   &ScaleSpriteSize(Scale Factor=&Vector2DFromFloat(Value=&FloatFromCurve(FloatCurve=DataInterface() {propertyValues {Curve {keys [{arriveTangent=1.9077, interpMode=RCIM_Cubic, leaveTangent=1.9077, tangentMode=RCTM_User, time=-0.0046, value=0.5236}, <at 1.0011 1.6623>]}}})))
   &ScaleColor(
    Scale Mode=::NewEnumerator1
    Scale RGBA=&Vector4FromCurve(Vector4Curve=DataInterface() {propertyValues {
       AlphaCurve {keys [
         <at 0.1219 0.9976 RCIM_Cubic>
         {arriveTangent=-1.5247, interpMode=RCIM_Cubic, leaveTangent=-1.5247, tangentMode=RCTM_User, time=0.5968, value=0.3116}
         <at 0.9977 -0.0026>
        ]}
       BlueCurve {keys [<at -0.0011 0.9976 RCIM_Cubic>,<at 1.0011 0.9976>]}
       GreenCurve {keys [<at -0.0011 0.9976 RCIM_Cubic>,<at 1.0011 0.9976>]}
       RedCurve {keys [<at -0.0011 0.9976 RCIM_Cubic>,<at 1.0011 0.9976 RCIM_Cubic>]}
      }})
    ScaleRGBA=true
   )
   &SolveForcesAndVelocity(Clamp Velocity=false, Limit Acceleration=false)
  ]
  rendererClasses [&NiagaraSpriteRendererProperties]
  renderers [{rendererClass=&NiagaraSpriteRendererProperties}]
 }
}
summary {
 systemName=NS_VaultDoorOpen
 emitters [
  {emitterName=BurstSlowCylin, rendererClasses [&NiagaraSpriteRendererProperties]}
  {emitterName=BurstFastCylin, rendererClasses [&NiagaraSpriteRendererProperties]}
 ]
 userVariables []
}
