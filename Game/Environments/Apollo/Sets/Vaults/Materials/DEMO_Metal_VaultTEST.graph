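# DEMO_Metal_VaultTEST is a Default Lit surface material in the Apollo Vaults set, apparently a demo or test metal for vault pieces.
# The graph holds only four constant colour vectors in grey and near-white, with no textures, parameters or math, and no connections to material outputs shown.
# It is flagged for use with Nanite, skeletal meshes and mesh deformers.
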Material /Game/Environments/Apollo/Sets/Vaults/Materials/DEMO_Metal_VaultTEST
graph {
 properties {
  bUsedWithMeshDeformer=true
  bUsedWithNanite=true
  bUsedWithSkeletalMesh=true
  pixelDepthOffsetMode=PDOM_Legacy
  usedShadingModels=Default Lit
 }
 nodes [
  Constant3Vector() {constant=<rgba 0.5417 0.5417 0.5417 1>}
  Constant3Vector() {constant=<rgba 0.984 1 0.9876 1>}
  Constant3Vector() {constant=<rgba 1 1 0.998 1>}
  Constant3Vector() {constant=<rgba 0.5 0.5 0.5 1>}
 ]
}
