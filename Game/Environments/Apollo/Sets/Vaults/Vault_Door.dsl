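; Vault door actor, based on BuildingGameplayActorVaultDoor, that opens on the server and blocks pawn interaction and death.
; VaultOpen flushes net dormancy, fires a replicated gameplay cue and the opened event, then toggles GreenLight for the light change.
; An authority-only timeline feeds rotation and shader values to UpdateVaultDoorOpening, while a one-shot multicast activates bolt and door effects in delayed stages.
; VaultDoorOpen is an authority-checked cheat entry point.
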(event EventBeginPlay
  (bind _returnvalue (Rendering|Material|CreateDynamicMaterialInstance (Variables|Default|GetVaultDoor) 0 0 "None" :!class /Script/Engine.PrimitiveComponent))
  (Variables|BuildingGameplayActorVaultDoor|SetVaultDoorMID _returnvalue)
  (Variables|BuildingGameplayActorVaultDoor|SetDoorStartingRotation (Transformation|GetWorldRotation (Variables|Default|GetVaultDoor))))

(event Custom|VaultOpen
  (Networking|FlushNetDormancy)
  (Ability|ExecuteGameplayCueReplicated (Variables|Default|GetGC_OpenDoor))
  (Variables|Default|SetGreenLight true)
  (Default|CallVaultHasOpened)
  (CallFunction|OpeningTimeline)
  ;[1 1 1] turn off light
  (Utilities|FlowControl|Delay 12.0)
  (Variables|Default|SetGreenLight false))

(event Custom|OpeningTimeline
  ;[0 0.9465 1] Using a timeline to rotate Door on open
  (Networking|SwitchHasAuthority
    (:Authority
      (CallFunction|MulticastPlayDoorOpenEffects)
      (bind (_direction _lerp _shaderanim_normalized) (Timeline|Open Play)
        (:Update
          (CallFunction|UpdateVaultDoorOpening _lerp _shaderanim_normalized))))))

(event Custom|Multicast_PlayDoorOpenEffects
  ;[1 0 0.9099] Multicast_PlayDoorOpenEffects
  (Utilities|FlowControl|DoOnce false
    (:Completed
      (Components|Activation|Activate (Variables|Default|GetUpperBolt) false)
      (Utilities|FlowControl|Delay (Variables|Default|GetWaitLockTime))
      (Components|Activation|Activate (Variables|Default|GetUpperBolt2_Door) false)
      (Utilities|FlowControl|Delay (Variables|Default|GetBigCShapeLockTime))
      (Components|Activation|Activate (Variables|Default|GetVaultDoorFX) false))))

(fn BlueprintCanInteract
  (params (InteractingPawn :type Fort%20Pawn%20Object%20Reference) (InteractionBeingAttempted :type EInteractionBeingAttempted%20Enum :default "FirstInteraction") (InteractionType :type TInteractionType%20Enum :default "IT_NoInteraction"))
  (return false))

(fn OnRep_HasBeenOpening
  (if (Variables|Default|GetGreenLight)
    (Rendering|Components|Light|SetLightColor (Variables|Default|GetVaultLightOutside) "(R=0.000000,G=1.000000,B=0.287746,A=1.000000)" true)
    (bind _returnvalue (Rendering|Material|CreateDynamicMaterialInstance (Variables|Default|GetFrame) 1 0 "None" :!class /Script/Engine.PrimitiveComponent))
    (Rendering|Material|SetVectorParameterValue _returnvalue "Emissive" "(R=0.000000,G=50.000000,B=4.914451,A=1.000000)" :!class /Script/Engine.MaterialInstanceDynamic)
    (else
      (Rendering|SetVisibility (Variables|Default|GetVaultLightInside) false false))))

(fn ShouldDie
  (params (Damage :type Float%20%28single-precision%29) (EventInstigator :type Controller%20Object%20Reference) (DamageCauser :type Actor%20Object%20Reference) (bIsTest :type Boolean) (EffectContext :type Gameplay%20Effect%20Context%20Handle%20Structure%20%28by%20ref%29))
  (return false))

(fn ConstructionScript
  (|CallParentFunction)
  (Variables|BuildingGameplayActorVaultDoor|SetDoorStaticMeshComponent (Variables|Default|GetVaultDoor))
  (Variables|BuildingGameplayActorVaultDoor|SetPrismStaticMeshComponent (Variables|Default|GetS_Prim_Prism)))

(fn VaultDoorOpen
  ;[1 0.9823 0.2193] cheat ke * VaultDoorOpen
  (Networking|SwitchHasAuthority
    (:Authority
      (CallFunction|VaultOpen))))
